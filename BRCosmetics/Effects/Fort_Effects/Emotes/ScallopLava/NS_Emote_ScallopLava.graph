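# Niagara system for the ScallopLava emote that draws paired red lava effects.
# Main_Energy layers three mesh renderers at two offset points, with material parameters driven by user PositionLerp values and a front-end preview override, while Decal001 adds matching decals and lights.
# Glow sprites and GPU LavaSplash flipbook sprites follow two skeletal mesh bones or sockets, picked by particle partition.
# Time-of-day inputs adjust glow, decal and energy brightness or color.
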NiagaraSystem /BRCosmetics/Effects/Fort_Effects/Emotes/ScallopLava/NS_Emote_ScallopLava
dependencies {
 usedDataInterfaces [
  &NiagaraDataInterfaceArrayFloat3
  &NiagaraDataInterfaceCurve
  &NiagaraDataInterfaceSpriteRendererInfo
  &NiagaraDataInterfaceArrayPosition
  &NiagaraDataInterfaceColorCurve
  &NiagaraDataInterfaceSkeletalMesh
 ]
 usedDynamicInputs []
 usedModules [
  &self:SystemSpawnScript.SystemScriptSource.SystemScriptGraph.NiagaraNodeAssignment_1.SetVariables_#1
  &SystemState
 ]
}
emitters {
 Decal001 {
  emitterUpdateScript [
   &EmitterState(
    Inactive Response=::NewEnumerator1
    Life Cycle Mode=::NewEnumerator1
    Loop Behavior=::NewEnumerator1
    Loop Duration Mode=::NewEnumerator1
    UseLoopDelay=false
   )
   &SpawnBurst_Instantaneous(Spawn Count=2, Use Spawn Probability=false, UseLoopCountLimit=false)
  ]
  particleSpawnScript [
   &PartitionParticles()
   &InitializeParticle(
    Color=<rgba 0.5885 0.0116 0 1>
    Lifetime=3
    Position Offset=/Niagara/DynamicInputs/Arrays/SelectPositionFromArray(
     Array Sampling Mode:ENiagaraArraySamplingMode=::NewEnumerator1
     Direct Array Index:NiagaraInt32=$Particles.PartitionParticles.Partition
     Interpolation:NiagaraFloat=?
     Direct Array Mode:ENiagara_WrapClamp=::NewEnumerator0
     Evaluation Type:ENiagaraRandomnessEvaluation=::NewEnumerator0
     Normalized Interpolation:NiagaraBool=true
     Recalculate Random Each Loop:NiagaraBool=false
     Position Array:NiagaraDataInterfaceArrayPosition=DataInterface() {propertyValues {GpuSyncMode=SyncCpuToGpu, MaxElements=0, PositionData [<-30 30 0>, <30 30 0>]}}
     Randomness Mode:ENiagaraRandomnessMode=::NewEnumerator0
     Fixed Random Seed:NiagaraBool=false
     Random Seed:NiagaraInt32=?
    )
    UsePositionOffset=true
    Write Material Random=true
   )
   /Niagara/Modules/Spawn/Decal/InitialDecalOrientation(
    Orientation Vector:Vector3f=<0 0 1>
    Rotation:Vector3f=&RandomRangeVector(Minimum=<0 0 0>, Maximum=<0 0 0>)
    Use Rotation Vector=true
   )
  ]
  particleUpdateScript [
   &ParticleState(Kill Particles When Lifetime Has Elapsed=false, Let Infinitely Lived Particles Die When Emitter Deactivates=true)
   /Niagara/Modules/Decal/Decal_Attributes(
    Color=1
    Color Mode=0
    Decal Color:LinearColor=&FNTimeOfDayColor(
     Night Color=<rgba 0.1927 0.0137 0.0137 1>
     Morning Color=<rgba 0.3 0.0213 0.0213 1>
     Day Color=<rgba 0.4844 0.0344 0.0344 1>
     Evening Color=<rgba 0.3 0.0213 0.0213 1>
    )
    Decal Fade:NiagaraFloat=&FloatFromCurve(
     FloatCurve=DataInterface() {propertyValues {Curve {keys [<at 0 0>, {arriveTangent=1.25, leaveTangent=1.25, time=0.2, value=1}, {arriveTangent=-1.25, leaveTangent=-1.25, time=0.5, value=1}, <at 1 0>]}}}
     Scale Curve=0.75
    )
    Decal Size:Vector3f=&Multiply_VectorByFloat(Vector=<1 1 1>, Float=270)
    Fade=1
    Position=1
    Size=1
    UsePositionOffset=true
   )
   &Light_Attributes(
    Color=1
    Color Value To Scale:LinearColor=&ColorFromCurve(ColorCurve=DataInterface() {propertyValues {
       AlphaCurve {keys []}
       BlueCurve {keys [<at 0.0196 0.1845 RCIM_Cubic>]}
       GreenCurve {keys [<at 0.0196 2.6691 RCIM_Cubic>]}
       RedCurve {keys [<at 0.0196 30 RCIM_Cubic>]}
      }})
    Enabled=1
    Falloff Exponent=1
    Light Exponent=3
    Light Position Offset:Vector3f=<0 0 30>
    Light Radius=130
    Position=1
    Radius=1
    Scale RGB:Vector3f=&Multiply_VectorByFloat(
     Vector=<1 1 1>
     Float=&FloatFromCurve(FloatCurve=DataInterface() {propertyValues {Curve {keys [<at 0 0>,<at 0.1343 1>,<at 0.4958 1>,<at 1 0>]}}}, Scale Curve=0.5)
    )
    ScaleA=true
    ScaleRGB=true
    UsePositionOffset=true
   )
  ]
  rendererClasses [&NiagaraDecalRendererProperties, &NiagaraLightRendererProperties]
  renderers [
   {rendererClass=&NiagaraDecalRendererProperties, rendererIndex=0}
   {rendererClass=&NiagaraLightRendererProperties, rendererIndex=1}
  ]
 }
 Glow {
  emitterUpdateScript [
   &EmitterState(Life Cycle Mode=::NewEnumerator1, Loop Behavior=::NewEnumerator1, Loop Duration=2, UseLoopDelay=true)
   &SpawnBurst_Instantaneous(Spawn Count=2, Spawn Time=0.1, Use Spawn Probability=false, UseLoopCountLimit=false)
  ]
  particleSpawnScript [
   &PartitionParticles()
   &InitializeParticle(
    Color=<rgba 1.5 0 0 0.5565>
    Lifetime=2
    Sprite Size Mode=::NewEnumerator3
    Uniform Sprite Size=100
    UsePositionOffset=false
    Write Material Random=true
   )
  ]
  particleUpdateScript [
   &ParticleState()
   &SkeletalMeshLocation(
    Bone / Socket Index=$Particles.PartitionParticles.Partition
    Bone Sampling Mode=::NewEnumerator4
    MaskSpawnGroup=false
    Sampled Position Offset=&SelectVectorFromArray(
     Direct Array Index=$Particles.PartitionParticles.Partition
     Vector Selection Array=DataInterface() {propertyValues {FloatData [<10 0 0>, <-10 0 0>]}}
    )
    Skeletal Mesh=$System.Skeletal Mesh Reader
   )
   &ScaleColor(
    Scale Alpha=&FloatFromCurve(
     FloatCurve=DataInterface() {propertyValues {Curve {
        defaultValue=1
        keys [
         <at 0 0 RCIM_Cubic>
         {arriveTangent=1.3186, interpMode=RCIM_Cubic, leaveTangent=1.3186, time=0.15, value=1}
         {arriveTangent=-0.0405, interpMode=RCIM_Cubic, leaveTangent=-0.0405, tangentMode=RCTM_User, time=0.6258, value=0.8253}
         {arriveTangent=-1.4695, interpMode=RCIM_Cubic, leaveTangent=-1.4695, tangentMode=RCTM_User, time=1, value=0}
        ]
       }}}
     Scale Curve=&Multiply_Float(A=0.5, B=&FN_Select_Float_from_Time_Of_Day(Morning=1, Day=1, Evening=1, Night=0.3))
    )
    ScaleA=true
    ScaleRGB=true
   )
   &CameraOffset(Camera Offset Amount=25)
  ]
  rendererClasses [&NiagaraSpriteRendererProperties]
  renderers [{rendererClass=&NiagaraSpriteRendererProperties}]
 }
 LavaSplash {
  simTarget=GPUComputeSim
  emitterUpdateScript [
   &EmitterState(Life Cycle Mode=::NewEnumerator1, Loop Behavior=::NewEnumerator1, Loop Duration=2, UseLoopDelay=false)
   &SpawnRate(
    SpawnRate=&FloatFromCurve(FloatCurve=DataInterface() {propertyValues {Curve {keys [<at 0.07 0>,<at 0.16 1>]}}}, CurveIndex=$Emitter.NormalizedLoopAge, Scale Curve=60)
    Use Spawn Probability=false
   )
  ]
  particleSpawnScript [
   &PartitionParticles()
   &InitializeParticle(
    Color=<rgba 1.5 0 0.3362 0.6522>
    Lifetime=0.6
    Sprite Rotation Mode=::NewEnumerator2
    Sprite Size Mode=::NewEnumerator1
    Uniform Sprite Size Max=60
    Uniform Sprite Size Min=30
    UsePositionOffset=false
    Write Material Random=true
   )
   &ShapeLocation(MaskSpawnGroup=false, Override Seed=false, Sphere Radius=15, Sphere Surface Distribution=1, UseOwnerScale=true)
   &AddVelocity(Module.Use Falloff Distance=false, Override Seed=false, Velocity Mode=::NewEnumerator1, Velocity Speed=-25)
   &CameraOffset(Camera Offset Amount=-30)
  ]
  particleUpdateScript [
   &SkeletalMeshLocation(
    Bone / Socket Index=$Particles.PartitionParticles.Partition
    Bone Sampling Mode=::NewEnumerator4
    MaskSpawnGroup=false
    Position Sampling=::NewEnumerator2
    Sampled Position Offset=&SelectVectorFromArray(
     Direct Array Index=$Particles.PartitionParticles.Partition
     Vector Selection Array=DataInterface() {propertyValues {FloatData [<10 0 0>, <-10 0 0>]}}
    )
    Skeletal Mesh=$System.Skeletal Mesh Reader
   )
   &ParticleState()
   &ScaleSpriteSize(Scale Sprite Size Mode=::NewEnumerator0, Uniform Scale Factor=0.6)
   &SpriteRotationRate(Rotation Rate=&RandomRangeFloat(Minimum=-200, Maximum=200))
   &Drag(Drag=-1, Use Linear Drag=true, Use Rotational Drag=false) {enabled=false}
   &GravityForce(Gravity=<0 0 -50>) {enabled=false}
   &SubUVAnimation(
    Evaluation Type=::NewEnumerator0
    "Sprite Renderer "=DataInterface() {propertyValues {SpriteRenderer=&self:LavaSplash_0.NiagaraSpriteRendererProperties_0}}
    SubUV Animation Mode:ENiagara_SubUVLookupModeV2=::NewEnumerator2
    UseEndFrame=false
    UseStartFrame=false
   )
   &ScaleVelocity(Velocity Scale=&Multiply_VectorByFloat(
     Vector=<1 1 1>
     Float=&FloatFromCurve(FloatCurve=DataInterface() {propertyValues {Curve {keys [<at 0 0.4>,<at 1 1>]}}}, Scale Curve=5)
    ))
   &SolveForcesAndVelocity(Clamp Velocity=false, Limit Acceleration=false)
   &ScaleColor(
    Scale RGB=&Multiply_VectorByFloat(
     Vector=<1 1 1>
     Float=&FloatFromCurve(FloatCurve=DataInterface() {propertyValues {Curve {keys [
          {tangentMode=RCTM_User, time=0, value=2}
          {arriveTangent=-27.8571, arriveTangentWeight=2.274, leaveTangent=-27.8571, leaveTangentWeight=4.2302, time=0.3, value=20}
         ]}}})
    )
    ScaleA=true
    ScaleRGB=true
   )
   &DynamicMaterialParameters(
    Index 0 Param 1=&FloatFromCurve(FloatCurve=DataInterface() {propertyValues {Curve {keys [
         <at 0 -0.1 RCIM_Cubic>
         {arriveTangent=0.75, interpMode=RCIM_Cubic, leaveTangent=0.75, time=0.3, value=0}
         {arriveTangent=0.7857, interpMode=RCIM_Cubic, leaveTangent=0.7857, time=0.4, value=0.2}
         <at 1 0.55>
        ]}}})
    Index 0 Param 2=&FloatFromCurve(FloatCurve=DataInterface() {propertyValues {Curve {keys [<at 0 3>,<at 0.3 1>]}}})
    Module.Index0Float0Write=true
    Module.Index0Float1Write=true
    Module.Index0Float2Write=true
    Module.Index0Float3Write=true
   )
   /SetVariables(Particles.Position02:NiagaraPosition=&AddVectorToPosition(Position=$Particles.SkeletalMeshLocation.SampledPosition, Vector=Linked() {linkedVariable {name=Particles.Position, type=NiagaraPosition}})) {isSetParameters=true}
  ]
  rendererClasses [&NiagaraSpriteRendererProperties]
  renderers [{rendererClass=&NiagaraSpriteRendererProperties}]
 }
 Main_Energy {
  emitterUpdateScript [
   &EmitterState(
    Inactive Response=::NewEnumerator1
    Life Cycle Mode=::NewEnumerator1
    Loop Behavior=::NewEnumerator1
    Loop Duration Mode=::NewEnumerator1
    UseLoopDelay=false
   )
   &SpawnBurst_Instantaneous(Spawn Count=2, Use Spawn Probability=false, UseLoopCountLimit=false)
  ]
  particleSpawnScript [
   &PartitionParticles()
   &InitializeParticle(
    Lifetime=3
    Mesh Scale=<0.7 0.7 1>
    Mesh Scale Mode=::NewEnumerator4
    Position Offset=&MakeCustomVectorFromBool(
     Bool=$User.bIsFrontEndPreview
     True Vector=<0 20 -10>
     False Vector=&SelectPositionFromArray(
      Array Sampling Mode:ENiagaraArraySamplingMode=::NewEnumerator1
      Direct Array Index:NiagaraInt32=$Particles.PartitionParticles.Partition
      Interpolation:NiagaraFloat=?
      Direct Array Mode:ENiagara_WrapClamp=::NewEnumerator0
      Evaluation Type:ENiagaraRandomnessEvaluation=::NewEnumerator0
      Normalized Interpolation:NiagaraBool=true
      Recalculate Random Each Loop:NiagaraBool=false
      Position Array:NiagaraDataInterfaceArrayPosition=DataInterface() {propertyValues {GpuSyncMode=SyncCpuToGpu, MaxElements=0, PositionData [<-30 30 0>, <30 30 0>]}}
      Randomness Mode:ENiagaraRandomnessMode=::NewEnumerator0
      Fixed Random Seed:NiagaraBool=false
      Random Seed:NiagaraInt32=?
     )
    )
    UsePositionOffset=true
    Write Material Random=true
   )
   &InitialMeshOrientation(Mesh Orientation Mode=::NewEnumerator2, Rotation=<0 0 0.06>, Use Rotation Vector=true)
   &CameraOffset(Camera Offset Amount=-5)
  ]
  particleUpdateScript [
   &ParticleState()
   &ScaleColor(
    Scale Alpha=&FloatFromCurve(
     FloatCurve=DataInterface() {propertyValues {Curve {keys [<at 0 0>,<at 0.1548 0.9412>,<at 0.6354 0.9647>,<at 1 0>]}}}
     Scale Curve=&Multiply_Float(
      A=&MakeCustomFloatFromBool(Bool=$User.bIsFrontEndPreview, True Float=0.8, False Float=1)
      B=&FNTimeOfDayScalar(Night Scale=1, Morning Scale=1, Evening Scale=1)
     )
    )
    Scale RGB=&MakeVectorFromLinearColorRGB(Color=<rgba 0.0885 0.0219 0.0123 1>)
    ScaleA=true
    ScaleRGB=true
   )
   /SetVariables(Particles.RedSize:Vector3f=<0.9 0.9 1.5>) {isSetParameters=true}
   /SetVariables(Particles.Color_Bottom:LinearColor=&MakeLinearColorFromVectorAndFloat(
     "Vector (RGB)"=&MakeVectorFromLinearColorRGB(Color=<rgba 0.1 0.0182 0.0182 1>)
     "Float (Alpha)"=&FloatFromCurve(
      FloatCurve=DataInterface() {propertyValues {Curve {keys [<at 0 0>,<at 0.1 1>,<at 0.6838 1>,<at 0.9982 0>]}}}
      Scale Curve=&Multiply_Float(
       A=&MakeCustomFloatFromBool(Bool=$User.bIsFrontEndPreview, True Float=0.5, False Float=1)
       B=&FNTimeOfDayScalar(Night Scale=0.5, Morning Scale=0.6, Evening Scale=0.5)
      )
     )
    )) {isSetParameters=true}
   /SetVariables(Particles.ColorLight:LinearColor=&MakeLinearColorFromVectorAndFloat(
     "Vector (RGB)"=&MakeVectorFromLinearColorRGB(Color=&FNTimeOfDayColor(Night Color=<rgba 0.1823 0 0 1>, Morning Color=<rgba 0.23 0 0 1>, Day Color=<rgba 0.4 0.02 0.02 1>, Evening Color=<rgba 0.25 0 0 1>))
     "Float (Alpha)"=&FloatFromCurve(
      FloatCurve=DataInterface() {propertyValues {Curve {keys [<at 0 0>,<at 0.1 1>,<at 0.9 1>,<at 1 1>]}}}
      Scale Curve=&Multiply_Float(A=1, B=&MakeCustomFloatFromBool(Bool=$User.bIsFrontEndPreview, True Float=0.3, False Float=1))
     )
    )) {isSetParameters=true}
   &DynamicMaterialParameters(
    Index 0 Param 1=2
    Index 0 Param 2=1.5
    Index 1 Param 1=&Lerp_Float(A=0, B=15, Alpha=$User.PositionLerp)
    Index 1 Param 2=&Lerp_Float(A=0, B=-10, Alpha=$User.PositionLerp)
    Index 1 Param 3=&Lerp_Float(A=1, B=12, Alpha=$User.PositionLerp_Z)
    Module.Index0Float0Write=true
    Module.Index0Float1Write=true
    Module.Index0Float2Write=true
    Module.Index0Float3Write=true
    Module.Index1Float0Write=true
    Module.Index1Float1Write=true
    Module.Index1Float2Write=true
    Module.Index1Float3Write=true
    Write Parameter Index 1=::NewEnumerator1
   )
  ]
  rendererClasses [&NiagaraMeshRendererProperties]
  renderers [
   {rendererClass=&NiagaraMeshRendererProperties, rendererIndex=0}
   {rendererClass=&NiagaraMeshRendererProperties, rendererIndex=1}
   {rendererClass=&NiagaraMeshRendererProperties, rendererIndex=2}
  ]
 }
}
summary {
 systemName=NS_Emote_ScallopLava
 emitters [
  {emitterName=Glow, rendererClasses [&NiagaraSpriteRendererProperties]}
  {emitterName=LavaSplash, simTarget=GPUComputeSim, rendererClasses [&NiagaraSpriteRendererProperties]}
  {emitterName=Main_Energy, rendererClasses [&NiagaraMeshRendererProperties]}
  {emitterName=Decal001, rendererClasses [&NiagaraDecalRendererProperties, &NiagaraLightRendererProperties]}
 ]
 userVariables [
  User.bIsFrontEnd {type=NiagaraBool}
  User.bIsFrontEndPreview {type=NiagaraBool}
  User.PositionLerp {type=NiagaraFloat, defaultValue=0, description=""}
  User.PositionLerp_Z {type=NiagaraFloat, defaultValue=0, description=""}
 ]
}
