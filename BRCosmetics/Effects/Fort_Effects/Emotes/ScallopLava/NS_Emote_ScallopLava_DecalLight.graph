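# Niagara system for the ScallopLava emote that shows a ground-style decal with a warm, red-dominant dynamic light.
# One burst-spawn emitter feeds both a decal renderer and a light renderer from the same particle.
# Curves over the roughly four-second lifetime fade the decal in and out and make the light flash bright, hold, then die.
# Two extra position attributes track the emitter origin, one of them raised slightly above it.
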NiagaraSystem /BRCosmetics/Effects/Fort_Effects/Emotes/ScallopLava/NS_Emote_ScallopLava_DecalLight
dependencies {
 usedDataInterfaces [&NiagaraDataInterfaceCurve, &NiagaraDataInterfaceColorCurve]
 usedDynamicInputs []
 usedModules [&SystemState]
}
emitters {Decal {
  emitterUpdateScript [
   &EmitterState(Inactive Response=::NewEnumerator1, Life Cycle Mode=::NewEnumerator1, Loop Behavior=::NewEnumerator1, Loop Duration=3.8, UseLoopDelay=true)
   &SpawnBurst_Instantaneous(Use Spawn Probability=false, UseLoopCountLimit=false)
  ]
  particleSpawnScript [
   &InitializeParticle(Color=<rgba 1 0.2158 0.059 1>, Lifetime=4, UsePositionOffset=true, Write Material Random=true)
   /Niagara/Modules/Spawn/Decal/InitialDecalOrientation(Orientation Vector:Vector3f=<0 0 1>, Use Rotation Vector=false)
  ]
  particleUpdateScript [
   /Niagara/Modules/Decal/Decal_Attributes(
    Color=1
    Decal Fade:NiagaraFloat=&FloatFromCurve(FloatCurve=DataInterface() {propertyValues {Curve {keys [
         <at 0 0 RCIM_Cubic>
         {arriveTangent=1.2359, leaveTangent=1.2359, time=0.3, value=1}
         {arriveTangent=-1.4283, leaveTangent=-1.4283, time=0.8, value=0.7658}
         {arriveTangent=-2.5528, leaveTangent=-2.5528, time=0.9, value=0}
         <at 1 0>
        ]}}})
    Decal Size:Vector3f=&Multiply_VectorByFloat(Vector=<5 5 2>, Float=65)
    Fade=1
    Scale RGB:Vector3f=&Multiply_VectorByFloat(
     Vector=<1 1 1>
     Float=&FloatFromCurve(FloatCurve=DataInterface() {propertyValues {Curve {keys [<at 0 1>,<at 0.5 1>,<at 1 0>]}}}, Scale Curve=0)
    )
    ScaleA=true
    ScaleRGB=true
    Size=1
   )
   &ParticleState()
   /SetVariables(Particles.Position2:NiagaraPosition=&AddVectorToPosition(Position=$Engine.Emitter.SimulationPosition, Vector=<0 0 0>)) {isSetParameters=true}
   /SetVariables(Particles.Position3:NiagaraPosition=&AddVectorToPosition(Position=$Engine.Emitter.SimulationPosition, Vector=<0 0 20>)) {isSetParameters=true}
   &Light_Attributes(
    Color=1
    Color Value To Scale:LinearColor=&ColorFromCurve(ColorCurve=DataInterface() {propertyValues {
       AlphaCurve {keys []}
       BlueCurve {keys [<at 0.0196 0 RCIM_Cubic>]}
       GreenCurve {keys [<at 0.0196 1 RCIM_Cubic>]}
       RedCurve {keys [<at 0.0196 30 RCIM_Cubic>]}
      }})
    Enabled=1
    Falloff Exponent=1
    Light Exponent=3
    Light Radius=120
    Radius=1
    Scale RGB:Vector3f=&Multiply_VectorByFloat(
     Vector=<1 1 1>
     Float=&FloatFromCurve(FloatCurve=DataInterface() {propertyValues {Curve {keys [<at 0.05 4>,<at 0.12 1>,<at 0.8 1>,<at 0.9 0>]}}}, Scale Curve=1.5)
    )
    ScaleA=true
    ScaleRGB=true
   )
  ]
  rendererClasses [&NiagaraDecalRendererProperties, &NiagaraLightRendererProperties]
  renderers [
   {rendererClass=&NiagaraDecalRendererProperties, rendererIndex=0}
   {rendererClass=&NiagaraLightRendererProperties, rendererIndex=1}
  ]
 }}
summary {
 systemName=NS_Emote_ScallopLava_DecalLight
 emitters [{emitterName=Decal, rendererClasses [&NiagaraDecalRendererProperties, &NiagaraLightRendererProperties]}]
 userVariables []
}
